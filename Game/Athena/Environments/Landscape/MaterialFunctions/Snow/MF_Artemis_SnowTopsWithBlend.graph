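# Landscape material function for Artemis that blends snow tops onto incoming material attributes, gated by the UseSnowTopsWithBlend static switch.
# Snow attributes come from MF_Artemis_SnowTops, with their normal angle-corrected over the strength-scaled source normal.
# MF_Fresnel_FX supplies the blend alpha, taking a SnowBlendVector parameter (up by default) instead of the camera vector, with adjustable brightness and exponent.
# The blend keeps the input's vertex attributes.
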MaterialFunction /Game/Athena/Environments/Landscape/MaterialFunctions/Snow/MF_Artemis_SnowTopsWithBlend
graph {nodes [
  FunctionOutput(None=StaticSwitchParameter(True=NamedRerouteUsage(), False=NamedRerouteUsage()) {group=SnowTopsWithBlend, parameterName=UseSnowTopsWithBlend})
  NamedRerouteDeclaration(None=FunctionInput() {inputName=MA, inputType=FunctionInput_MaterialAttributes}) {name=MA_In, nodeColor=<rgba 0.5234 1 0 1>}
  n1 = GetMaterialAttributes(None=NamedRerouteUsage()) {attributeGetTypes [#1]}
  n2 = GetMaterialAttributes(None=MaterialFunctionCall() {materialFunction=&/MF_Artemis_SnowTops}.Result) {attributeGetTypes [#1]}
  n3 = MaterialFunctionCall(
   BaseNormal=Multiply(A=@n1.Normal, B=VectorParameter() {group=SnowTopsWithBlend, parameterName=SourceNormalStrength, defaultValue=<rgba 1 1 1 0>}.RGB)
   AdditionalNormal=@n2.Normal
  ) {materialFunction=&BlendAngleCorrectedNormals}
  n4 = ScalarParameter() {defaultValue=1, group=SnowTopsWithBlend, parameterName=SnowTopBlendBrightness}
  n5 = ScalarParameter() {defaultValue=1, group=SnowTopsWithBlend, parameterName=SnowTopBlendExponent}
  n6 = VectorParameter() {group=SnowTopsWithBlend, parameterName=SnowBlendVector, defaultValue=<rgba 0 0 1 0>}
  NamedRerouteDeclaration(None=BlendMaterialAttributes(
    A=@n1.MaterialAttributes
    B=SetMaterialAttributes!(MaterialAttributes=@n2.MaterialAttributes, Normal=@n3.Result) {attributeSetTypes [#1]}
    Alpha=MaterialFunctionCall(Brightness=@n4, Exponent=@n5, Camera Vector=@n6.RGB) {materialFunction=&MF_Fresnel_FX}.Fresnel
   ) {vertexAttributeBlendType=UseA}) {name=Blended, nodeColor=<rgba 0 1 0.4844 1>}
 ]}
