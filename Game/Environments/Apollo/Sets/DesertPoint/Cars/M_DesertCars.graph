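# Master material for DesertPoint car props that grounds them in the Apollo landscape by blending in runtime virtual texture color and normals by height above terrain.
# A Textured switch picks texture maps with dust and optional waterline wetness, or a vertex-colored look with masked noise detail.
# HeightLerp breaks up the terrain blend, then shared functions add snow tops and destructible damage.
# Forward, mobile, lower-quality and proxy paths use cheaper substitutes.
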Material /Game/Environments/Apollo/Sets/DesertPoint/Cars/M_DesertCars
graph {
 properties {bTangentSpaceNormal=false, bUseMaterialAttributes=true, pixelDepthOffsetMode=PDOM_Legacy, usedShadingModels=Default Lit}
 nodes [
  n1 = TextureObjectParameter() {parameterName=Diffuse, texture=/Game/Vehicles/FORT_Vehicles_01/Materials/T_Camper_D}
  n2 = TextureObjectParameter() {parameterName=SpecularMasks, samplerType=SAMPLERTYPE_Masks, texture=&T_Fortnite_Default_S}
  n3 = TextureObjectParameter() {parameterName=Normals, samplerType=SAMPLERTYPE_Normal, texture=/Engine/EngineMaterials/FlatNormal}
  n4 = MaterialFunctionCall!(
   Diffuse Texture Object=@n1
   Spec Texture Object=@n2
   TextureSize=ScalarParameter() {defaultValue=512, parameterName=WorldSize}
   --------------------------
   Use seperate settings for Z projection
   Use seperate texture for Z projection
   Texture Size Z projection
   Z Plane Diffuse Texture Object
   Z Plane Spec Texture Object
   "--------------------------      "
   ProjectionTransitionContrast
   Use Texture Alpha Heightmap to Refine the Transistion
   HeightLerpTransitionContrast
   PickTheDiffuseHeightChannel
   TheDiffuseHeightChannel
   "--------------------------       "
   NormalTextureObject=@n3
   World Space Normals=StaticBool()
   "--------------------------  "
   Export Float 4 Diffuse=StaticBool() {value=true}
   Export Float 4 Spec
   "--------------------------     "
   WorldPosition
   World Space Normal
  ) {materialFunction=/Engine/Functions/Engine_MaterialFunctions02/Texturing/WorldAlignedTextures_Complex}
  n5 = TextureSample(Tex=@n1) {texture=&T_Fortnite_Default_S}
  n6 = StaticSwitchParameter(True=ComponentMask(None=@n4.Z Diffuse Texture) {b=true, g=true, r=true}, False=@n5.RGB) {group=Textures, parameterName=WS Z Texture}
  n7 = StaticSwitchParameter(True=ComponentMask(None=@n4.Z Diffuse Texture) {a=true}, False=@n5.A) {group=Textures, parameterName=WS Z Texture}
  n8 = LinearInterpolate(A=@n6, B=Multiply(A=VectorParameter() {group=Textures, parameterName=Textured Mod color, defaultValue=<rgba 1 1 1 0>}.RGB, B=@n6), Alpha=@n7)
  n9 = StaticSwitchParameter(
   True=@n4.Normal Z Texture
   False=TextureSample(Tex=@n3) {samplerType=SAMPLERTYPE_Normal, texture=&T_Fortnite_Default_S}.RGB
  ) {group=Textures, parameterName=WS Z Texture}
  n10 = Transform(None=@n9)
  n11 = Saturate(None=Multiply(A=ComponentMask(None=@n10) {b=true}, B=ScalarParameter() {defaultValue=1, parameterName=DustLevel}))
  n12 = RuntimeVirtualTextureSampleParameter() {materialType=BaseColor_Normal_Specular_YCoCg, parameterName=RVT, virtualTexture=/Game/Athena/Apollo/Maps/Landscape/VT/RVT_Apollo_Landscape}
  n13 = MaterialProxyReplace(Realtime=@n12.BaseColor, MaterialProxy=VectorParameter() {parameterName=ProxyColor, defaultValue=<rgba 0.3854 0.3028 0.169 0>}.RGB)
  n14 = Subtract(
   A=ComponentMask(None=WorldPosition().XYZ) {b=true}
   B=RuntimeVirtualTextureSampleParameter() {materialType=WorldHeight, parameterName=RVTH, virtualTexture=/Game/Athena/Apollo/Maps/Landscape/VT/RVT_Apollo_Landscape_Height}.WorldHeight
  )
  n15 = Divide(A=@n14, B=ScalarParameter() {defaultValue=55, parameterName=RVT Metal R Fade})
  n16 = Saturate(None=Divide(A=@n15))
  n17 = Saturate(None=DotProduct(A=VertexNormalWS(), B=Constant3Vector() {constant=<rgba 0 0 -1 0>}))
  n18 = MaterialFunctionCall(Float3=StaticSwitchParameter(
    True=@n4.Z Spec
    False=TextureSample(Tex=@n2) {samplerType=SAMPLERTYPE_Masks, texture=&T_Fortnite_Default_S}.RGB
   ) {group=Textures, parameterName=WS Z Texture}) {materialFunction=&BreakOutFloat3Components}
  n19 = SetMaterialAttributes!(
   MaterialAttributes
   Base Color=LinearInterpolate(A=@n13, B=LinearInterpolate(A=@n8, B=@n13, Alpha=@n11), Alpha=MaterialProxyReplace(Realtime=@n16, MaterialProxy=@n17))
   Metallic=@n18.G
   Roughness=LinearInterpolate(A=@n18.B, Alpha=@n11)
   Normal=@n9
   Specular=@n18.G
  ) {attributeSetTypes [
    #1
    #2
    #3
    #4
    #5
   ]}
  n20 = SetMaterialAttributes(Base Color=@n8, Metallic=@n18.G, Roughness=@n18.B, Normal=@n9, Specular=@n18.B) {attributeSetTypes [
    #1
    #2
    #3
    #4
    #5
   ]}
  n21 = ShadingPathSwitch(Default=@n19, Deferred=@n19, Forward=@n20, Mobile=@n20)
  n22 = GetMaterialAttributes(None=@n21) {attributeGetTypes [#1, #5, #3]}
  n23 = Multiply(A=@n22.Base Color, B=VectorParameter() {group=Wet, parameterName=Wet, defaultValue=<rgba 1 1 1 0>}.RGB) {constB=0.3}
  n24 = Multiply(A=@n22.Specular) {constB=0.3}
  n25 = Multiply(A=@n22.Roughness) {constB=0.2}
  n26 = Subtract(A=WorldPosition().XYZ, B=ActorPositionWS())
  n27 = Divide(
   A=Add(A=ComponentMask(None=@n26) {b=true}, B=ScalarParameter() {defaultValue=-512, group=Wet, parameterName=WaterlineOffset})
   B=ScalarParameter() {group=Wet, parameterName=Waterline Blur}
  )
  n28 = StaticSwitchParameter(
   True=BlendMaterialAttributes(
    A=SetMaterialAttributes!(MaterialAttributes=@n22.MaterialAttributes, Base Color=@n23, Specular=@n24, Roughness=@n25) {attributeSetTypes [#1, #5, #3]}
    B=@n21
    Alpha=Saturate(None=@n27)
   )
   False=@n21
  ) {parameterName=UseWaterline}
  n29 = Constant3Vector() {constant=<rgba 0.3906 0.3906 0.3906 0>}
  n30 = Multiply(A=VertexColor(), B=@n29)
  n31 = Saturate(None=ComponentMask(None=VertexNormalWS()) {b=true})
  n32 = Constant() {r=0.7}
  n33 = DotProduct(A=VertexNormalWS(), B=Constant3Vector() {constant=<rgba 0 0 -1 0>})
  n34 = ComponentMask(None=WorldPosition().XYZ) {g=true, r=true}
  n35 = ScalarParameter() {defaultValue=5, parameterName=DetailsScale}
  n36 = TextureCoordinate()
  n37 = TextureSample(UVs=StaticSwitchParameter(True=Divide(A=@n34, B=@n35), False=Multiply(A=@n36, B=@n35)) {group=Textures, parameterName=WS Z Details}) {samplerType=SAMPLERTYPE_Masks, texture=/Game/Accessories/FORT_Backpacks/M_MED_Heist/Textures/T_CrystalNoise_M}
  n38 = MaterialFunctionCall(Transition Phase=MaterialProxyReplace(Realtime=Saturate(None=@n15), MaterialProxy=Saturate(None=@n33)), Height Texture=@n37.R) {materialFunction=&HeightLerp}
  n39 = TextureSampleParameter2D() {
   group=Textures
   parameterName="Mask(B)"
   samplerType=SAMPLERTYPE_Masks
   texture=/Game/Environments/Apollo/Props/Vehicles/Standard_Car_Junker/Textures/T_Standard_Car_Junker_FX
  }
  n40 = Constant() {r=0.5}
  n41 = Constant() {r=0.8}
  n42 = LinearInterpolate(A=Constant() {r=0.3}, B=Constant() {r=0.45}, Alpha=@n37.R)
  n43 = Constant() {r=0.2}
  n44 = TextureSampleParameter2D() {
   group=Textures
   parameterName=Normal
   samplerType=SAMPLERTYPE_Normal
   texture=/Game/Environments/Apollo/Props/Vehicles/Standard_Car_Junker/Textures/T_Standard_Car_Junker_Mossy_N
  }
  n45 = SetMaterialAttributes!(
   MaterialAttributes
   Base Color=LinearInterpolate(A=LinearInterpolate(A=@n30, B=@n13, Alpha=@n31), B=LinearInterpolate(A=@n13, B=@n32, Alpha=@n38.Lerp Alpha No Contrast), Alpha=@n39.B)
   Metallic=LinearInterpolate(B=@n39.B, Alpha=@n38.Lerp Alpha No Contrast)
   Roughness=LinearInterpolate(A=@n40, B=LinearInterpolate(A=@n41, B=@n42, Alpha=@n38.Lerp Alpha No Contrast), Alpha=@n39.B)
   Normal=@n44.RGB
   Specular=Add(A=@n43, B=@n37.R)
  ) {attributeSetTypes [
    #1
    #2
    #3
    #4
    #5
   ]}
  n46 = SetMaterialAttributes(
   Base Color=LinearInterpolate(
    A=LinearInterpolate(A=@n29, B=@n32, Alpha=@n39.B)
    B=VectorParameter() {parameterName=DustMobile, defaultValue=<rgba 0.3359 0.2641 0.1626 0>}.RGB
    Alpha=@n31
   )
   Metallic=@n39.B
   Roughness=LinearInterpolate(A=@n41, B=@n40, Alpha=@n39.B)
   Normal=@n44.RGB
   Specular=@n43
  ) {attributeSetTypes [
    #1
    #2
    #3
    #4
    #5
   ]}
  n47 = MaterialFunctionCall!(Material=StaticSwitchParameter(
    True=QualitySwitch(Default=@n28, Low=@n21, High=@n28, Medium=@n21, Epic=@n28)
    False=ShadingPathSwitch(Default=@n45, Deferred=@n45, Forward=@n46, Mobile=@n46)
   ) {parameterName=Textured}) {materialFunction=/Game/Athena/Apollo/Maps/Landscape/VT/MaterialFunctions/TransformTangentToWorld}
  n48 = LinearInterpolate(
   A=ScalarParameter() {defaultValue=10, parameterName=Blend RVT Sides}
   B=ScalarParameter() {defaultValue=56, parameterName="Blend RVT "}
   Alpha=@n31
  ) {constB=0}
  n49 = LinearInterpolate(
   A=ScalarParameter() {defaultValue=1, group=Textures, parameterName=AlphaMax}
   B=ScalarParameter() {group=Textures, parameterName=AlphaMin}
   Alpha=@n7
  )
  n50 = ScalarParameter() {defaultValue=0.5, parameterName=To RVT Contrast}
  n51 = BlendMaterialAttributes(
   A=SetMaterialAttributes!(MaterialAttributes, Base Color=@n12.BaseColor, Specular=@n12.BaseColor, Roughness=@n12.BaseColor) {attributeSetTypes [#1, #5, #3]}
   B=@n47.Result
   Alpha=MaterialFunctionCall(
    Transition Phase=Divide(A=@n14, B=@n48)
    Height Texture=StaticSwitchParameter(True=@n49, False=@n37.R) {parameterName=Use DiffuseAlpha}
    Contrast=LinearInterpolate(A=@n50, Alpha=@n31) {constB=0}
   ) {materialFunction=&HeightLerp}.Alpha
  ) {vertexAttributeBlendType=UseA}
  n52 = LinearInterpolate(
   A=ScalarParameter() {defaultValue=3, parameterName=Blend RVT Normals Sides}
   B=ScalarParameter() {defaultValue=4, parameterName=Blend RVT Normals}
   Alpha=@n31
  ) {constB=0}
  n53 = BlendMaterialAttributes(
   A=SetMaterialAttributes!(MaterialAttributes=@n51, Normal=@n12.Normal) {attributeSetTypes [#4]}
   B=@n51
   Alpha=Saturate(None=Divide(A=@n14, B=@n52))
  ) {vertexAttributeBlendType=UseA}
  n54 = MaterialFunctionCall(
   NoRVT=@n47.Result
   RVT=MaterialProxyReplace(Realtime=QualitySwitch(Default=@n53, Low=@n51, High=@n53, Medium=@n51, Epic=@n53), MaterialProxy=@n47.Result)
  ) {materialFunction=&FN_RVT_Switch_Material}
  MaterialFunctionCall(
   Material_In=MaterialFunctionCall!(MA=ShadingPathSwitch(Default=@n54.Result, Deferred=@n54.Result, Forward=@n47.Result, Mobile=@n47.Result)) {materialFunction=/Game/Athena/Environments/Landscape/MaterialFunctions/Snow/MF_Artemis_SnowTopsWithBlend}.Result
   UseCrackedDamageTextures=StaticBool()
   UseAdditiveWPO=StaticBool() {value=true}
  ) {materialFunction=&MF_FN_DestructibleDamage}
 ]
}
